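; Blueprint interface for the Mask Maker boss challenge manager, declaring messages about the boss fight's state.
; It declares functions for the boss being stunned, the boss recovering from stun, and a weakpoint health update that carries the weakpoint's index and its health value.
; The file holds only signatures, so the response to each message is left to implementers.
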(fn BossRecoveredFromStun)

(fn BossStunned)

(fn WeakpointHealthUpdated
  (params (WeakpointIndex :type Integer) (WeakpointHealth :type Float%20%28double-precision%29)))
