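; Blueprint interface for the MaskMaker boss pawn, declaring the functions other code uses to query and drive the boss without casting to it.
; Getters and setters expose fight state such as floating, stun, fast rise, unaware idle, Spire Slam use, beam end and last target location.
; Action calls cover weak point toggling, ability montages, cooldowns, rise-to-floating, spire shard spawning, damage-resist effect removal, portal rotation and aggro switching.
; It holds only signatures with default returns, no logic.
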(fn ClearAndInvalidateRiseToFloatingTimer)

(fn ClearSavedSpireShardsLocations)

(fn ClearSavedTargetedPlayers)

(fn GetBeamEndLocation
  (return "0, 0, 0"))

(fn GetChallengerManagerRef
  (return 0))

(fn GetDoFastRise
  (return false))

(fn GetDoUnawareIdle
  (return false))

(fn GetIsEveryMaskDisabled
  (return false))

(fn GetIsFloating
  (return false))

(fn GetIsInStunnedState
  (return false))

(fn GetJustFloatedUp
  (return false))

(fn GetJustUsedSpireSlam
  (return false))

(fn GetMaskWeakpoints
  (return 0))

(fn GetMaxTeamSize
  (return 0))

(fn GetPortalTargetPoint
  (return 0))

(fn GetSavedSpireShardsLocations
  (return 0))

(fn GetSavedTargetedPlayers
  (return 0))

(fn GetSpawnLocation
  (return "0, 0, 0"))

(fn GetTargetLastLocation
  (return "0, 0, 0"))

(fn GetTeleportAreaBox
  (return 0))

(fn GetTotalShardSpawnedAmount
  (return 0))

(fn MaskMakerActivateCriticalWeakPoints
  (params (Activate :type Boolean)))

(fn MaskMakerActivateJustFloatedUpCd)

(fn MaskMakerCallBossRecoveredFromStun)

(fn MaskMakerCallSpireShardAbilityStarted)

(fn MaskMakerDisableAllWeakPoints)

(fn MaskMakerEnableAllWeakPoints)

(fn MaskMakerInitiateWeakPointsData)

(fn MaskMakerPlayAbilityAnim
  (params (AnimMontage :type Anim%20Montage%20Object%20Reference) (PlayRate :type Float%20%28double-precision%29 :default 1.0) (SectionName :type Name :default "Default"))
  (return))

(fn MaskMakerRemoveDamageResistGE)

(fn MaskMakerRotateTowardPortal)

(fn MaskMakerSpawnSpireShardsPlayers)

(fn MaskMakerStartBetweenAbilityCd)

(fn MaskMakerStartRiseToFloating
  (params (Montage :type Anim%20Montage%20Object%20Reference) (AnimLenght :type Float%20%28double-precision%29) (AnimPlayRate :type Float%20%28double-precision%29 :default 1.0) (AnimSectionName :type Name :default "Default")))

(fn SetBeamEndLocation
  (params (Location :type Vector)))

(fn SetChallengeManagerRef
  (params (ChallengeManager :type Actor%20Object%20Reference)))

(fn SetDamageInStunnedState
  (params (DamageAmount :type Float%20%28double-precision%29)))

(fn SetDmgResistGE
  (params (GEHandle :type Active%20Gameplay%20Effect%20Handle%20Structure)))

(fn SetDoFastRise
  (params (Bool :type Boolean)))

(fn SetDoUnawareIdle
  (params (Bool :type Boolean)))

(fn SetIsFloating
  (params (Bool :type Boolean))
  (return))

(fn SetIsInStunnedState
  (params (Bool :type Boolean)))

(fn SetJustFloatedUp
  (params (Bool :type Boolean)))

(fn SetJustUsedSpireSlam
  (params (Bool :type Boolean)))

(fn SetMaskMakerLocation
  (params (Location :type Vector)))

(fn SetShardSpawnedAmount
  (params (ShardAmount :type Integer)))

(fn SetTargetLastLocation)

(fn SetUpdateTargetLastLoc
  (params (Bool :type Boolean)))

(fn TrySwitchAggroTargetEvent)
