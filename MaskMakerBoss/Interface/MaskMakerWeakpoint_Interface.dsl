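; Blueprint interface for weakpoints on the MaskMaker boss.
; It declares one function, GetWeakpointCurrentHealth, which returns a float for the weakpoint's current health.
; The declaration holds only a zero default, so any health logic lives in the classes that implement it.
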(fn GetWeakpointCurrentHealth
  (return 0.0))
